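# Niagara system for the HonorBraceLeap emote, showing flying likes, a running counter and speed lines around the player.
# Likes loops bursts driven by user curve floats, sending red sprites, meshes and lights along -Y with pops at selectable collect points.
# Counter follows character sockets, counting at an age-curved rate blended by owner speed and front-end state.
# SpeedLines streams fast streaks, partitioned across attach, calf and head sockets with height taken from the spine.
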NiagaraSystem /BRCosmetics/Animation/Game/MainPlayer/Emotes/HonorBraceLeap/FX/NS_Emote_HonorBraceLeap_Likes
dependencies {
 usedDataInterfaces [
  &NiagaraDataInterfaceArrayFloat3
  &NiagaraDataInterfaceCurve
  &NiagaraDataInterfaceColorCurve
  &NiagaraDataInterfaceVector2DCurve
  &NiagaraDataInterfaceSocketReader
  &NiagaraDataInterfaceArrayInt32
  &NiagaraDataInterfaceVectorCurve
  &NiagaraDataInterfaceCamera
 ]
 usedDynamicInputs []
 usedModules [
  &SystemState
  &self:SystemSpawnScript.SystemScriptSource.SystemScriptGraph.NiagaraNodeAssignment_0.SetVariables_#1
  &self:SystemSpawnScript.SystemScriptSource.SystemScriptGraph.NiagaraNodeAssignment_1.SetVariables_#2
 ]
}
emitters {
 Counter {
  emitterUpdateScript [
   &EmitterState(
    Inactive Response=::NewEnumerator1
    Life Cycle Mode=::NewEnumerator1
    Loop Behavior=::NewEnumerator1
    Loop Duration Mode=::NewEnumerator1
    UseLoopDelay=false
   )
   &self:InterpolateSpeedThresholdEmitter_02(
    Define Initial Value=false
    Rate Of Change=5
    Target Float Value:NiagaraFloat=&MakeCustomFloatFromBool(
     Bool=&SetBoolByFloatComparison(
      A=&Add_Float(
       A=&Add_Float(
        A=&Abs_Float(Float:NiagaraFloat=&MakeFloatFromVector(Vector=$Engine.Owner.Velocity, Channel=::NewEnumerator0))
        B=&Abs_Float(Float:NiagaraFloat=&MakeFloatFromVector(Vector=$Engine.Owner.Velocity, Channel=::NewEnumerator2))
       )
       B=&Abs_Float(Float:NiagaraFloat=&MakeFloatFromVector(Vector=$Engine.Owner.Velocity, Channel=::NewEnumerator1))
      )
      B=46.278
     )
     True Float=1
     False Float=0
    )
   ) {module=InterpolateSpeedThresholdEmitter_02}
   &SpawnBurst_Instantaneous(Use Spawn Probability=false, UseLoopCountLimit=false)
  ]
  particleSpawnScript [
   &PartitionParticles()
   &InitializeParticle(
    Lifetime=&MergeHelper_FloatSwitch(0=0, 1=0, 2=0, 3=0, 4=0, Index=0, Num=::NewEnumerator3)
    Position Offset=<0 50.8098 120.0871>
    Sprite Rotation Mode=::NewEnumerator0
    Sprite Size Mode=::NewEnumerator3
    Uniform Sprite Size=44.2106
    UsePositionOffset=true
    Write Material Random=true
   )
   &DynamicMaterialParameters(Module.Index0Float0Write=true, Module.Index0Float1Write=true, Module.Index0Float2Write=true, Module.Index0Float3Write=true)
  ]
  particleUpdateScript [
   &ParticleState(Kill Particles When Lifetime Has Elapsed=false)
   &ScaleSpriteSize(
    Uniform Curve Index=$Emitter.Age
    Uniform Curve Sprite Scale=DataInterface() {propertyValues {Curve {
       defaultValue=1
       keys [
        {arriveTangent=4.1785, interpMode=RCIM_Cubic, leaveTangent=4.1785, tangentMode=RCTM_User, time=0, value=0}
        {arriveTangent=1.1179, interpMode=RCIM_Cubic, leaveTangent=1.1179, tangentMode=RCTM_User, time=0.0988, value=0.7182}
        {arriveTangent=0.0224, interpMode=RCIM_Cubic, leaveTangent=0.0224, tangentMode=RCTM_User, time=0.3, value=1}
       ]
      }}}
   )
   &SocketLocation(
    Position Sampling=::NewEnumerator1
    SamplePosOffset=true
    Sampled Position Offset=<-20 20 0>
    Socket Reader=DataInterface() {propertyValues {
      AttachComponentClass=null
      AttachComponentTag=None
      EditorPreviewAsset=/Game/Characters/Player/Male/Male_Avg_Base/Fortnite_M_Avg_Player
      SourceActorRef=null
      SourceAsset=null
      SourceMode=Default
      bRequireCurrentFrameData=false
      bUpdateSocketsPerFrame=true
      FilteredSockets [AttachSocket]
      ObjectParameterBinding {parameter {name=None, typeDefHandle {registeredTypeIndex=1}, varData []}}
     }}
    Socket Sampling Mode=::NewEnumerator3
    VelocitySpeedLimit=false
   )
   &SocketLocation(
    Position Sampling=::NewEnumerator1
    SamplePosOffset=true
    Sampled Position Offset=<0 0 79.0615>
    Socket Reader=DataInterface() {propertyValues {
      AttachComponentClass=null
      AttachComponentTag=None
      EditorPreviewAsset=&Fortnite_M_Avg_Player
      SourceActorRef=null
      SourceAsset=null
      SourceMode=Default
      bRequireCurrentFrameData=false
      bUpdateSocketsPerFrame=true
      FilteredSockets [FX_Spine]
      ObjectParameterBinding {parameter {name=None, typeDefHandle {registeredTypeIndex=1}, varData []}}
     }}
    Socket Sampling Mode=::NewEnumerator3
    VelocitySpeedLimit=false
   ) {module=SocketLocation001}
   /SetVariables(Particles.NewPos:NiagaraPosition=&MakeVector(
     X=&MakeFloatFromVector(Vector=Linked() {linkedVariable {name=Output.SocketLocation.SampledPosition, type=NiagaraPosition}}, Channel=::NewEnumerator0)
     Y=&MakeFloatFromVector(Vector=Linked() {linkedVariable {name=Output.SocketLocation.SampledPosition, type=NiagaraPosition}}, Channel=::NewEnumerator1)
     Z=&MakeFloatFromVector(Vector=Linked() {linkedVariable {name=Output.SocketLocation001.SampledPosition, type=NiagaraPosition}}, Channel=::NewEnumerator2)
    )) {isSetParameters=true, module=SetVariables_#3}
   &self:CounterTime(CounterSpeed:NiagaraFloat=&Lerp_Float(
     A=&FN_FrontEndFloatModifier(
      Gameplay=20
      Front End=&FloatFromCurve(
       FloatCurve=DataInterface() {propertyValues {Curve {
          defaultValue=1
          keys [
           {interpMode=RCIM_Cubic, tangentMode=RCTM_User, time=0, value=25}
           {arriveTangent=47.8786, interpMode=RCIM_Cubic, leaveTangent=47.8786, tangentMode=RCTM_User, time=5, value=100}
          ]
         }}}
       CurveIndex=$Emitter.Age
      )
     )
     B=&FloatFromCurve(
      FloatCurve=DataInterface() {propertyValues {Curve {
         defaultValue=1
         keys [
          {interpMode=RCIM_Cubic, tangentMode=RCTM_User, time=0, value=25}
          {arriveTangent=47.8786, interpMode=RCIM_Cubic, leaveTangent=47.8786, tangentMode=RCTM_User, time=5, value=100}
         ]
        }}}
      CurveIndex=$Emitter.Age
     )
     Alpha=$Emitter.InterpolateSpeedThresholdEmitter_02.InterpolateOverTime_Float.Moving Average
    )) {module=CounterTime001}
  ]
  rendererClasses [&NiagaraSpriteRendererProperties]
  renderers [{rendererClass=&NiagaraSpriteRendererProperties}]
 }
 Likes {
  emitterUpdateScript [
   /SetVariables(Emitter.SpeedDebug=1) {isSetParameters=true}
   &EmitterState(
    Life Cycle Mode=::NewEnumerator1
    Loop Duration=&Lerp_Float(A=0.3, B=0.23, Alpha=$Emitter.SpeedDebug)
    Recalculate Duration Each Loop=true
    UseLoopDelay=false
   )
   &SpawnBurst_Instantaneous(
    Spawn Count=&MakeIntFromFloat(Float=$User.SpawnCurve, Conversion Mode=::NewEnumerator1)
    Use Spawn Probability=false
    UseLoopCountLimit=false
   ) {module=SpawnBurst_Instantaneous002}
   &SpawnBurst_Instantaneous(
    Spawn Count=&MakeIntFromFloat(Float=$User.BurstCurve)
    Spawn Group=1
    Spawn Time=&Lerp_Float(A=0.1, B=0.17, Alpha=$Emitter.SpeedDebug)
    Use Spawn Probability=false
    UseLoopCountLimit=false
   ) {module=SpawnBurst_Instantaneous001}
   &SpawnBurst_Instantaneous(
    Spawn Count=&MergeHelper_IntSwitch(
     0=0
     1=&RandomRangeInt(Minimum=4, Maximum=7)
     2=?
     3=?
     4=?
     5=?
     Index=&MakeIntFromFloat(Float=$User.BurstCurve)
     Num=::NewEnumerator0
    )
    Spawn Group=2
    Spawn Time=&Lerp_Float(A=0.1, B=0.17, Alpha=$Emitter.SpeedDebug)
    Use Spawn Probability=false
    UseLoopCountLimit=false
   )
   &SpawnBurst_Instantaneous(
    Spawn Count=&MergeHelper_IntSwitch(
     0=0
     1=&RandomRangeInt(Minimum=10, Maximum=20)
     2=?
     3=?
     4=?
     5=?
     Index=&MakeIntFromFloat(Float=$User.BurstCurve)
     Num=::NewEnumerator0
    )
    Spawn Group=3
    Spawn Time=&Lerp_Float(A=0.1, B=0.17, Alpha=$Emitter.SpeedDebug)
    Use Spawn Probability=false
    UseLoopCountLimit=false
   ) {module=SpawnBurst_Instantaneous003}
   &SpawnBurst_Instantaneous(
    Spawn Count=&MergeHelper_IntSwitch(
     0=0
     1=&RandomRangeInt(Minimum=5, Maximum=8)
     2=?
     3=?
     4=?
     5=?
     Index=&MakeIntFromFloat(Float=$User.BurstCurve)
     Num=::NewEnumerator0
    )
    Spawn Group=4
    Spawn Time=&Lerp_Float(A=0.1, B=0.17, Alpha=$Emitter.SpeedDebug)
    Use Spawn Probability=false
    UseLoopCountLimit=false
   ) {module=SpawnBurst_Instantaneous004}
  ]
  particleSpawnScript [
   /SetVariables(Particles.VisibilityTag=&SpawnGroup()) {isSetParameters=true}
   &InitializeParticle(
    Color=<rgba 3 0.0313 0.0313 1>
    Lifetime=&MergeHelper_FloatSwitch(
     0=0.45
     1=0.2
     2=&RandomRangeFloat(Minimum=0.07, Maximum=0.1)
     3=0.16
     4=0.2
     Index=&SpawnGroup()
     Num=::NewEnumerator3
    )
    Mesh Scale Mode=::NewEnumerator4
    Position Offset=<0 167 80>
    Sprite Rotation Mode=::NewEnumerator2
    Sprite Size Mode=::NewEnumerator3
    Uniform Sprite Size=&MergeHelper_FloatSwitch(
     0=40
     1=78
     2=&RandomRangeFloat(Minimum=11.42, Maximum=17.8486)
     3=4
     4=0
     Index=&SpawnGroup()
     Num=::NewEnumerator3
    )
    UsePositionOffset=true
    Write Material Random=true
   )
   /SetVariables(Particles.Position=&SelectVectorFromArray(
     Direct Array Index=&MakeIntFromFloat(Float=$User.PosCurve)
     Interpolation=$User.PosCurve
     Normalized Interpolation=false
     Vector Selection Array=DataInterface() {propertyValues {FloatData [<0 160 100>, <30 160 100>, <-30 160 100>, <0 160 178>, <-20 160 100>, <0 160 100>]}}
    )) {isSetParameters=true}
   &ShapeLocation(MaskSpawnGroup=true, Override Seed=false, Spawn Group Mask=2, Sphere Radius=15.0361, UseOwnerScale=true)
   /SetVariables(Particles.NewPos:NiagaraPosition=&AddVectorToPosition(
     Position=&MaskVectorBySpawnGroup(Masked Vector=$Particles.ShapeLocation.ShapeVector, Passthrough Vector=<0 0 0>, Spawn Group Mask=2)
     Vector=&MergeHelper_VectorSwitch(
      0=<0 30 100>
      1=<30 30 100>
      2=<-30 30 100>
      3=<0 30 178>
      4=<-20 30 100>
      5=<0 30 100>
      Index=&MakeIntFromFloat(Float=$User.CollectPos, Conversion Mode=::NewEnumerator1)
      Num=::NewEnumerator4
     )
    )) {isSetParameters=true}
   &AddVelocity(Velocity=&MaskVectorBySpawnGroup(Masked Vector=<0 -300 0>, Passthrough Vector=<0 0 0>))
   &InitialMeshOrientation(
    Mesh Orientation Mode=::NewEnumerator7
    Rotation=&RandomRangeVector(Minimum=<0 0 0>, Maximum=<0 0 1>)
    Use Rotation Vector=true
   )
   &CameraOffset(Camera Offset Amount=&MergeHelper_FloatSwitch(0=0, 1=0, 2=-5, 3=0, 4=0, Index=&SpawnGroup(), Num=::NewEnumerator3))
   &PartitionParticles(Number Of Partitions=&MaskIntBySpawnGroup(Masked Int=2, Passthrough Int=0, Spawn Group Mask=3))
  ]
  particleUpdateScript [
   &ParticleState()
   &ScaleMeshSize(Scale Factor=&VectorFromFloat(Value=&FloatFromCurve(FloatCurve=DataInterface() {propertyValues {Curve {
         defaultValue=0
         keys [
          {arriveTangent=4.1785, interpMode=RCIM_Cubic, leaveTangent=4.1785, tangentMode=RCTM_User, time=0, value=0}
          {arriveTangent=-0.4611, interpMode=RCIM_Cubic, leaveTangent=-0.4611, tangentMode=RCTM_User, time=0.2042, value=1.2183}
          {arriveTangent=-0.2743, interpMode=RCIM_Cubic, leaveTangent=-0.2743, time=0.3086, value=1}
          {arriveTangent=0.0224, interpMode=RCIM_Cubic, leaveTangent=0.0224, tangentMode=RCTM_User, time=1, value=1}
         ]
        }}})))
   &UpdateMeshOrientation(Rotation Vector=<0 0 1>)
   &ScaleSpriteSize(
    Scale Sprite Size Mode=::NewEnumerator0
    Uniform Scale Factor=&MergeHelper_FloatSwitch(
     0=&FloatFromCurve(FloatCurve=DataInterface() {propertyValues {Curve {
         defaultValue=0
         keys [
          {arriveTangent=4.1785, interpMode=RCIM_Cubic, leaveTangent=4.1785, tangentMode=RCTM_User, time=0, value=0}
          {arriveTangent=-0.4611, interpMode=RCIM_Cubic, leaveTangent=-0.4611, tangentMode=RCTM_User, time=0.2042, value=1.2183}
          {arriveTangent=-0.2743, interpMode=RCIM_Cubic, leaveTangent=-0.2743, time=0.3086, value=1}
          {arriveTangent=0.0224, interpMode=RCIM_Cubic, leaveTangent=0.0224, tangentMode=RCTM_User, time=1, value=1}
         ]
        }}})
     1=&FloatFromCurve(FloatCurve=DataInterface() {propertyValues {Curve {
         defaultValue=0
         keys [
          {interpMode=RCIM_Cubic, tangentMode=RCTM_User, time=0, value=0}
          {interpMode=RCIM_Cubic, tangentMode=RCTM_User, time=0.5, value=1}
          {tangentMode=RCTM_User, time=1, value=0}
         ]
        }}})
     2=&FloatFromCurve(FloatCurve=DataInterface() {propertyValues {Curve {defaultValue=0, keys [{tangentMode=RCTM_User, time=0, value=1}, {tangentMode=RCTM_User, time=1, value=0}]}}})
     3=1.4924
     4=0
     Index=&SpawnGroup()
     Num=::NewEnumerator3
    )
   )
   &ScaleColor(
    Scale Mode=::NewEnumerator1
    Scale RGBA=/Game/Effects/Niagara/Modules/MergeEmitters/MergeHelper_Vector4Switch(
     0:Vector4f=&Vector4FromCurve(Vector4Curve=DataInterface() {propertyValues {
        AlphaCurve {keys [<at 0.9065 1>,<at 1 0>]}
        BlueCurve {keys [<at 0 1>,<at 1 1>]}
        GreenCurve {keys [<at 0 1>,<at 1 1>]}
        RedCurve {keys [<at 0 1>,<at 1 1>]}
       }})
     1:Vector4f=&Vector4FromCurve(Vector4Curve=DataInterface() {propertyValues {
        AlphaCurve {keys [<at 0 1 RCIM_Cubic>,<at 1 0 RCIM_Cubic>]}
        BlueCurve {keys [<at 0 0.125 RCIM_Cubic>,<at 1 0.0244 RCIM_Cubic>]}
        GreenCurve {keys [<at 0 0.125 RCIM_Cubic>,<at 1 0 RCIM_Cubic>]}
        RedCurve {keys [<at 0 3 RCIM_Cubic>,<at 1 0.5365 RCIM_Cubic>]}
       }})
     2:Vector4f=&Vector4FromCurve(Vector4Curve=DataInterface() {propertyValues {
        AlphaCurve {keys [<at 0 1 RCIM_Cubic>]}
        BlueCurve {keys [<at 0 0.125 RCIM_Cubic>]}
        GreenCurve {keys [<at 0 0.125 RCIM_Cubic>]}
        RedCurve {keys [<at 0 3 RCIM_Cubic>]}
       }})
     3:Vector4f=<1 1 1 1>
     4:Vector4f=<1 1 1 1>
     5:Vector4f=?
     6:Vector4f=?
     7:Vector4f=?
     8:Vector4f=?
     9:Vector4f=?
     Index:NiagaraInt32=0
     Num:MergeHelper_SwitcherEnum=::NewEnumerator3
    )
    ScaleRGBA=true
   )
   &DynamicMaterialParameters(
    Index 0 Param 1=&FloatFromCurve(FloatCurve=DataInterface() {propertyValues {Curve {
        defaultValue=1
        keys [
         {arriveTangent=4.1785, interpMode=RCIM_Cubic, leaveTangent=4.1785, tangentMode=RCTM_User, time=0, value=0}
         {arriveTangent=0.0224, interpMode=RCIM_Cubic, leaveTangent=0.0224, tangentMode=RCTM_User, time=0.2138, value=1}
        ]
       }}})
    Index 0 Param 2=&FloatFromCurve(FloatCurve=DataInterface() {propertyValues {Curve {
        defaultValue=1
        keys [
         {arriveTangent=4.1785, interpMode=RCIM_Cubic, leaveTangent=4.1785, tangentMode=RCTM_User, time=0, value=0}
         {arriveTangent=0.0224, interpMode=RCIM_Cubic, leaveTangent=0.0224, tangentMode=RCTM_User, time=0.2138, value=1}
        ]
       }}})
    Index 0 Param 3=&FloatFromCurve(FloatCurve=DataInterface() {propertyValues {Curve {
        defaultValue=1
        keys [
         {arriveTangent=4.1785, interpMode=RCIM_Cubic, leaveTangent=4.1785, tangentMode=RCTM_User, time=0, value=0}
         {arriveTangent=0.0224, interpMode=RCIM_Cubic, leaveTangent=0.0224, tangentMode=RCTM_User, time=0.2138, value=1}
        ]
       }}})
    Module.Index0Float0Write=true
    Module.Index0Float1Write=true
    Module.Index0Float2Write=true
    Module.Index0Float3Write=true
    Module.Index1Float0Write=true
    Module.Index1Float1Write=true
    Module.Index1Float2Write=true
    Module.Index1Float3Write=true
    Write Parameter Index 1=::NewEnumerator1
   )
   /SetVariables(Particles.CrossFlare Size:Vector2f=&Multiply_Vector2D_ByFloat(
     Float=&FloatFromCurve(
      FloatCurve=DataInterface() {propertyValues {Curve {
         defaultValue=1
         keys [
          {interpMode=RCIM_Cubic, tangentMode=RCTM_User, time=0, value=0}
          {arriveTangent=-6.4388, interpMode=RCIM_Cubic, leaveTangent=-6.4388, tangentMode=RCTM_User, time=0.4027, value=1.0194}
          {tangentMode=RCTM_User, time=1, value=0}
         ]
        }}}
      Scale Curve=1.7867
     )
     Vector2D=<35 35>
    )) {isSetParameters=true}
   /SetVariables(
    Particles.BGColo:LinearColor=<rgba 0.474 0 0.0211 0.9478>
    Particles.BGSize:Vector2f=&Multiply_Vector2D_ByFloat(
     Float=&FloatFromCurve(FloatCurve=DataInterface() {propertyValues {Curve {
         defaultValue=1
         keys [
          {interpMode=RCIM_Cubic, tangentMode=RCTM_User, time=0, value=0}
          {interpMode=RCIM_Cubic, tangentMode=RCTM_User, time=0.5, value=1}
          {tangentMode=RCTM_User, time=1, value=0}
         ]
        }}})
     Vector2D=<70 70>
    )
   ) {isSetParameters=true}
   &OffsetPosition(Position Offset=&MakeVector(
     X=0
     Y=0
     Z=&Waveform("Waveform [1]"=::NewEnumerator0, "[1] Frequency"=1.4758, "[1] Amplitude Min/Max"=<0 5.7581>)
    ))
   &ShapeLocation(
    Apply To Particle Position=::NewEnumerator1
    MaskSpawnGroup=true
    Override Seed=false
    Spawn Group Mask=3
    Sphere Radius=2
    Sphere Surface Distribution=1
    UseOwnerScale=true
   ) {module=ShapeLocation001}
   &ShapeLocation(
    Apply To Particle Position=::NewEnumerator1
    MaskSpawnGroup=true
    Spawn Group Mask=4
    Sphere Distribution=::NewEnumerator3
    Sphere Radius=&FloatFromCurve(
     FloatCurve=DataInterface() {propertyValues {Curve {
        defaultValue=1
        keys [
         {arriveTangent=1.8851, interpMode=RCIM_Cubic, leaveTangent=1.8851, tangentMode=RCTM_User, time=0.1, value=0.611}
         {arriveTangent=0.0224, interpMode=RCIM_Cubic, leaveTangent=0.0224, tangentMode=RCTM_User, time=1, value=1}
        ]
       }}}
     CurveIndex=$Emitter.NormalizedLoopAge
     Scale Curve=25.7294
    )
    Uniform Distribution:NiagaraFloat=&RandomRangeFloat(Minimum=0, Maximum=2)
    UseOwnerScale=true
   ) {module=ShapeLocation002}
   /SetVariables(
    Particles.Pop1Pos:NiagaraPosition=&AddVectorToPosition(Position=$Particles.ShapeLocation001.ShapePosition, Vector=Linked() {linkedVariable {name=Particles.NewPos, type=NiagaraPosition}})
    Particles.Pop2Pos:NiagaraPosition=&AddVectorToPosition(Position=$Particles.ShapeLocation002.ShapePosition, Vector=Linked() {linkedVariable {name=Particles.NewPos, type=NiagaraPosition}})
   ) {isSetParameters=true}
   /SetVariables(
    Particles.Dot2Colors:LinearColor=&ColorFromCurve(ColorCurve=DataInterface() {propertyValues {
       AlphaCurve {keys [<at 0.1806 0>,<at 0.3192 1 RCIM_Cubic>]}
       BlueCurve {keys [<at 0 1 RCIM_Cubic>,<at 1 1 RCIM_Cubic>]}
       GreenCurve {keys [<at 0 1 RCIM_Cubic>,<at 1 1 RCIM_Cubic>]}
       RedCurve {keys [<at 0 1 RCIM_Cubic>,<at 1 1 RCIM_Cubic>]}
      }})
    Particles.Dot2Size:Vector2f=&Vector2DFromCurve(
     Vector2Curve=DataInterface() {propertyValues {
       XCurve {keys [<at 0.1 1>,<at 1 1>]}
       YCurve {
        defaultValue=0
        keys [
         {arriveTangent=-2.4288, interpMode=RCIM_Cubic, leaveTangent=-2.4288, tangentMode=RCTM_User, time=0.0664, value=1}
         {arriveTangent=-0.0326, tangentMode=RCTM_User, time=1, value=0}
        ]
       }
      }}
     ScaleCurve=&RandomRangeVector2D(Minimum=<2 36.0144>, Maximum=<2 35.236>)
    )
    Particles.DotAlignment:Vector3f=$Particles.ShapeLocation001.ShapeNormal
    Particles.DotColors:LinearColor=&MergeHelper_ColorSwitch(0=<rgba 3 0 0 1>, 1=<rgba 0.1826 0.1826 0.1826 0.9391>, 2=?, 3=?, Index=$Particles.PartitionParticles.Partition, Num=::NewEnumerator0)
    Particles.DotRotation=180
    Particles.DotSize:Vector2f=&Vector2DFromCurve(
     Vector2Curve=DataInterface() {propertyValues {
       XCurve {keys [<at 0.1 1>,<at 1 0>]}
       YCurve {
        defaultValue=0
        keys [
         {arriveTangent=4.1785, interpMode=RCIM_Cubic, leaveTangent=4.1785, tangentMode=RCTM_User, time=0.1, value=0}
         {arriveTangent=1.1179, interpMode=RCIM_Cubic, leaveTangent=1.1179, tangentMode=RCTM_User, time=0.319, value=0.6228}
         {arriveTangent=0.0224, interpMode=RCIM_Cubic, leaveTangent=0.0224, tangentMode=RCTM_User, time=1, value=1}
        ]
       }
      }}
     ScaleCurve=&RandomRangeVector2D(Minimum=<1 15>, Maximum=<2 20>)
    )
    Particles.DotsAlignment2:Vector3f=$Particles.ShapeLocation002.ShapeNormal
   ) {isSetParameters=true}
   &FN_Light_Attributes(
    Color=&ColorFromCurve(ColorCurve=DataInterface() {propertyValues {
       AlphaCurve {keys [<at 0 1 RCIM_Cubic>,<at 0.7958 0 RCIM_Cubic>]}
       BlueCurve {keys [<at 0 0.131 RCIM_Cubic>]}
       GreenCurve {keys [<at 0 0 RCIM_Cubic>]}
       RedCurve {keys [<at 0 3 RCIM_Cubic>]}
      }})
    Create Color Attribute=true
    Create Position Attribute=true
    Create Radius Attribute=true
    Enable Visibility Tag=true
    Position:NiagaraPosition=$Particles.NewPos
    Radius=50
    VisibilityTag=1
   )
   &GravityForce(Gravity=&MaskVectorBySpawnGroup(Masked Vector=<0 0 -600>, Passthrough Vector=<0 0 0>, Spawn Group Mask=5))
   &SolveForcesAndVelocity(Clamp Velocity=false, Limit Acceleration=false)
  ]
  rendererClasses [
   &NiagaraSpriteRendererProperties
   &NiagaraMeshRendererProperties
   &NiagaraLightRendererProperties
  ]
  renderers [
   {rendererClass=&NiagaraSpriteRendererProperties}
   {rendererClass=&NiagaraSpriteRendererProperties, rendererIndex=1}
   {rendererClass=&NiagaraSpriteRendererProperties, rendererIndex=2}
   {rendererClass=&NiagaraSpriteRendererProperties, rendererIndex=3}
   {rendererClass=&NiagaraMeshRendererProperties, rendererIndex=4}
   {rendererClass=&NiagaraSpriteRendererProperties, rendererIndex=5}
   {rendererClass=&NiagaraLightRendererProperties, rendererIndex=6}
   {rendererClass=&NiagaraSpriteRendererProperties, rendererIndex=7}
   {rendererClass=&NiagaraSpriteRendererProperties, rendererIndex=8}
  ]
 }
 SpeedLines {
  emitterUpdateScript [
   &EmitterState(Life Cycle Mode=::NewEnumerator1, Loop Behavior=::NewEnumerator1, Loop Duration Mode=::NewEnumerator1, UseLoopDelay=false)
   &self:InterpolateSpeedThresholdEmitter_02(
    Define Initial Value=false
    Rate Of Change=5
    Target Float Value:NiagaraFloat=&MakeCustomFloatFromBool(
     Bool=&SetBoolByFloatComparison(
      A=&Add_Float(
       A=&Add_Float(
        A=&Abs_Float(Float:NiagaraFloat=&MakeFloatFromVector(Vector=$Engine.Owner.Velocity, Channel=::NewEnumerator0))
        B=&Abs_Float(Float:NiagaraFloat=&MakeFloatFromVector(Vector=$Engine.Owner.Velocity, Channel=::NewEnumerator2))
       )
       B=&Abs_Float(Float:NiagaraFloat=&MakeFloatFromVector(Vector=$Engine.Owner.Velocity, Channel=::NewEnumerator1))
      )
      B=46.278
     )
     True Float=1
     False Float=0
    )
   ) {enabled=false, module=InterpolateSpeedThresholdEmitter_02}
   &SpawnRate(
    SpawnRate=&Lerp_Float(A=&FN_FrontEndFloatModifier(Gameplay=70, Front End=90), B=150, Alpha=1)
    Use Spawn Probability=false
   )
   &SpawnRate(
    SpawnGroup=1
    SpawnRate=&Lerp_Float(A=&FN_FrontEndFloatModifier(Gameplay=20, Front End=30), B=60, Alpha=1)
    Use Spawn Probability=false
   ) {module=SpawnRate001}
   /SetVariables(Emitter.Thickness:NiagaraFloat=&RandomRangeFloat(Minimum=0.25, Maximum=0.4)) {isSetParameters=true}
  ]
  particleSpawnScript [
   /SetVariables(Particles.VisibilityTag=&SpawnGroup()) {isSetParameters=true}
   &PartitionParticles(Number Of Partitions=6)
   &InitializeParticle(
    Color=&MergeHelper_ColorSwitch(0=<rgba 0 0 0 1>, 1=<rgba 2.7188 2.7188 2.7188 0.75>, 2=?, 3=?, Index=&SpawnGroup(), Num=::NewEnumerator0)
    Lifetime=&RandomRangeFloat(Minimum=0.22, Maximum=0.1)
    Position Offset=<0 0 41.2251>
    Sprite Rotation Mode=::NewEnumerator0
    Sprite Size=&MergeHelper_Vector2Switch(
     0=/Niagara/DynamicInputs/Helpers/MaskVector2d(
      Mask Source:NiagaraInt32=1
      Mask Parameter:NiagaraInt32=$Particles.PartitionParticles.Partition
      Mask:NiagaraInt32=4
      Masked Vector2D:Vector2f=&RandomRangeVector2D(Minimum=<20 161.1938>, Maximum=<25 103.071>)
      Passthrough Vector2D:Vector2f=&RandomRangeVector2D(Minimum=<45 161.1938>, Maximum=<30 103.071>)
     )
     1=<0.7581 60.4032>
     2=?
     3=?
     4=?
     Index=&SpawnGroup()
     Num=::NewEnumerator0
    )
    Sprite Size Mode=::NewEnumerator4
    UsePositionOffset=true
    Write Material Random=true
   )
   &SpriteFacingAndAlignment(WriteAlignment=true, WriteFacing=false)
   &AddVelocity(Velocity=<0 -13289.3311 0>)
  ]
  particleUpdateScript [
   &ParticleState()
   &ScaleSpriteSize(
    Non-Uniform Curve Scale=<1 1.5>
    Non-Uniform Curve Sprite Scale=DataInterface() {propertyValues {
      XCurve {keys [{tangentMode=RCTM_User, time=0.5969, value=1}, {tangentMode=RCTM_User, time=1.0029, value=0}]}
      YCurve {keys [{tangentMode=RCTM_User, time=0.0148, value=0}, {tangentMode=RCTM_User, time=0.9481, value=1}]}
     }}
    Scale Sprite Size Mode=::NewEnumerator3
   )
   &SocketLocation(
    Position Sampling=::NewEnumerator1
    SamplePosOffset=true
    Sampled Position Offset=&Lerp_Vector(
     A=&MergeHelper_VectorSwitch(0=<0 0 0>, 1=<0 0 0>, 2=<0 0 0>, 3=<0 0 0>, 4=<0 10 0>, Index=$Particles.PartitionParticles.Partition, Num=::NewEnumerator3)
     B=&MergeHelper_VectorSwitch(0=<0 -70 0>, 1=<0 -70 0>, 2=<0 -70 0>, 3=<0 70 0>, 4=<0 -70 0>, Index=$Particles.PartitionParticles.Partition, Num=::NewEnumerator3)
     Alpha=&FloatFromCurve(FloatCurve=DataInterface() {propertyValues {Curve {defaultValue=1, keys [{tangentMode=RCTM_User, time=0, value=0}, {tangentMode=RCTM_User, time=1, value=1}]}}})
    )
    Socket Index=&SelectIntFromArray(Direct Array Index=$Particles.PartitionParticles.Partition, Int Selection Array=DataInterface() {propertyValues {IntData [0,0,1,2,3,3]}})
    Socket Reader=DataInterface() {propertyValues {
      AttachComponentClass=null
      AttachComponentTag=None
      EditorPreviewAsset=&Fortnite_M_Avg_Player
      SourceActorRef=null
      SourceAsset=null
      SourceMode=Default
      bRequireCurrentFrameData=false
      bUpdateSocketsPerFrame=true
      FilteredSockets [AttachSocket,calf_l,calf_r,FX_Head]
      ObjectParameterBinding {parameter {name=None, typeDefHandle {registeredTypeIndex=1}, varData []}}
     }}
    Socket Sampling Mode=::NewEnumerator3
    VelocitySpeedLimit=false
   )
   &SocketLocation(
    Position Sampling=::NewEnumerator1
    SamplePosOffset=true
    Sampled Position Offset=&MergeHelper_VectorSwitch(0=<0 0 50>, 1=<0 0 50>, 2=<0 0 -50>, 3=<0 0 -50>, 4=<0 0 74.3846>, Index=$Particles.PartitionParticles.Partition, Num=::NewEnumerator3)
    Socket Reader=DataInterface() {propertyValues {
      AttachComponentClass=null
      AttachComponentTag=None
      EditorPreviewAsset=&Fortnite_M_Avg_Player
      SourceActorRef=null
      SourceAsset=null
      SourceMode=Default
      bRequireCurrentFrameData=false
      bUpdateSocketsPerFrame=true
      FilteredSockets [FX_Spine]
      ObjectParameterBinding {parameter {name=None, typeDefHandle {registeredTypeIndex=1}, varData []}}
     }}
    Socket Sampling Mode=::NewEnumerator3
    VelocitySpeedLimit=false
   ) {module=SocketLocation001}
   &ShapeLocation(
    Box Size=&MergeHelper_VectorSwitch(0=<10 0 81.4797>, 1=<10 0 9.2684>, 2=<0 0 0>, 3=<0 0 0>, 4=<0 0 0>, Index=$Particles.PartitionParticles.Partition, Num=::NewEnumerator3)
    MaskSpawnGroup=true
    Offset=&MergeHelper_VectorSwitch(
     0=<0 -8.3587 -68.907>
     1=<0 17.3025 -17.7905>
     2=<0 0 0>
     3=<0 0 0>
     4=<0 0 -25>
     Index=$Particles.PartitionParticles.Partition
     Num=::NewEnumerator3
    )
    Offset Mode=::NewEnumerator0
    Override Seed=false
    Shape Primitive=::NewEnumerator2
    Surface Only=false
    UseOwnerScale=true
   )
   &ShapeLocation(
    Additional Quaternion Rotation=false
    Box Size=&MergeHelper_VectorSwitch(0=<30 0 32.8204>, 1=<30 0 9.2684>, 2=<-1.0212 0 0>, 3=<30 0 0>, Index=$Particles.PartitionParticles.Partition, Num=::NewEnumerator2)
    MaskSpawnGroup=true
    Offset=&MergeHelper_VectorSwitch(0=<0 -8.3587 -68.907>, 1=<0 17.3025 -40>, 2=<0 0 0>, 3=<0 0 0>, Index=$Particles.PartitionParticles.Partition, Num=::NewEnumerator2)
    Offset Mode=::NewEnumerator0
    Override Seed=false
    Rotation Mode=::NewEnumerator3
    Rotation Quaternion=$Output.SocketLocation.SampledRotation
    Shape Primitive=::NewEnumerator2
    Spawn Group Mask=1
    Surface Expansion Mode:ENiagaraExpansionMode=::NewEnumerator2
    Surface Only=true
    Surface Only Band Thickness=6.2635
    UseOwnerScale=true
   ) {module=ShapeLocation001}
   &ScaleColor(
    Scale Alpha=&MergeHelper_FloatSwitch(
     0=1
     1=&FloatFromCurve(FloatCurve=DataInterface() {propertyValues {Curve {keys [<at 0 1>,<at 1 0>]}}})
     2=?
     3=?
     Index=&SpawnGroup()
     Num=::NewEnumerator0
    )
    ScaleA=true
    ScaleRGB=false
   ) {enabled=false}
   /SetVariables(Particles.NewPos:NiagaraPosition=&AddVectorToPosition(
     Position=$Particles.Position
     Vector=&MakeVector(
      X=&MakeFloatFromVector(Vector=Linked() {linkedVariable {name=Output.SocketLocation.SampledPosition, type=NiagaraPosition}}, Channel=::NewEnumerator0)
      Y=&MakeFloatFromVector(Vector=Linked() {linkedVariable {name=Output.SocketLocation.SampledPosition, type=NiagaraPosition}}, Channel=::NewEnumerator1)
      Z=&MakeFloatFromVector(Vector=Linked() {linkedVariable {name=Output.SocketLocation001.SampledPosition, type=NiagaraPosition}}, Channel=::NewEnumerator2)
     )
    )) {isSetParameters=true, module=SetVariables_#3}
   &SolveForcesAndVelocity(Clamp Velocity=false, Limit Acceleration=false)
   /SetVariables(Particles.FaceCameraLerp:Vector3f=&Lerp_Vector(
     A=<0 0 1>
     B=<1 0 0>
     Alpha=&DotProduct(
      Vector A=&VectorFromCurve(
       VectorCurve=DataInterface() {propertyValues {XCurve {keys [<at -0.5 1>,<at 0 0>,<at 0.7415 1>]}, YCurve {keys [<at 0.7984 1>,<at 1 0>]}, ZCurve {keys [<at 0.7984 1>,<at 1 0>]}}}
       CurveIndex=&Abs_Float(Float:NiagaraFloat=&MakeFloatFromVector(
         Vector=&ReturnCameraVector(Camera Vectors=::NewEnumerator1, Coordinate Space=::Local)
         Channel=::NewEnumerator0
        ))
       Scale Curve=<1 0 0>
      )
      Vector B=<1 0 0>
     )
    )) {enabled=false, isSetParameters=true}
  ]
  rendererClasses [&NiagaraSpriteRendererProperties]
  renderers [
   {rendererClass=&NiagaraSpriteRendererProperties}
   {rendererClass=&NiagaraSpriteRendererProperties, rendererIndex=1}
  ]
 }
}
summary {
 systemName=NS_Emote_HonorBraceLeap_Likes
 emitters [
  {
   emitterName=Likes
   rendererClasses [
    &NiagaraSpriteRendererProperties
    &NiagaraMeshRendererProperties
    &NiagaraLightRendererProperties
   ]
  }
  {emitterName=Counter, rendererClasses [&NiagaraSpriteRendererProperties]}
  {emitterName=SpeedLines, rendererClasses [&NiagaraSpriteRendererProperties]}
 ]
 userVariables [
  User.bIsFrontEnd {type=NiagaraBool}
  User.bIsFrontEndPreview {type=NiagaraBool}
  User.BurstCurve {type=NiagaraFloat, defaultValue=1, description=""}
  User.CollectPos {type=NiagaraFloat, defaultValue=0, description=""}
  User.PosCurve {type=NiagaraFloat, defaultValue=0, description=""}
  User.SpawnCurve {type=NiagaraFloat, defaultValue=1, description=""}
 ]
}
